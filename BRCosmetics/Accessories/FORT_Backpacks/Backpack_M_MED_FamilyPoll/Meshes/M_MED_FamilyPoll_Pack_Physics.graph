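# Physics asset for the FamilyPoll backpack cosmetic, giving its dangling strap a simple simulated chain beside the pack body.
# Capsule bodies cover the pack, a tiny base_strap anchor and three dyn_strap segments.
# Each strap segment is constrained to the one before it with locked twist and narrow swing limits, while base_strap joins the pack with free swing and twist.
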PhysicsAsset /BRCosmetics/Accessories/FORT_Backpacks/Backpack_M_MED_FamilyPoll/Meshes/M_MED_FamilyPoll_Pack_Physics
bodies {
 base_strap [{shapeType=Capsule, length=0.505, radius=0.505, shapeName=base_strap_capsule, rotation=<rot 0 -90 0>}]
 dyn_strap_01 [{shapeType=Capsule, length=1, radius=1.5, shapeName=dyn_strap_01_capsule, center=<2.5 0 0>}]
 dyn_strap_02 [{shapeType=Capsule, length=6, radius=1.3, shapeName=dyn_strap_02_capsule, center=<2.2 -0.3 0>}]
 dyn_strap_03 [{shapeType=Capsule, length=6, radius=1.2, shapeName=dyn_strap_03_capsule, center=<3 -0.5 0>}]
 pack [{shapeType=Capsule, length=33, radius=11, center=<0 10 -20>}]
}
constraints [
 {
  bone1Name=base_strap
  bone2Name=pack
  swing1LimitDegrees=45
  swing1Motion=Free
  swing2LimitDegrees=45
  swing2Motion=Free
  twistLimitDegrees=45
  twistMotion=Free
 }
 {bone1Name=dyn_strap_01, bone2Name=base_strap, swing1LimitDegrees=15, swing2LimitDegrees=25, twistLimitDegrees=1, twistMotion=Locked}
 {bone1Name=dyn_strap_02, bone2Name=dyn_strap_01, swing1LimitDegrees=20, swing2LimitDegrees=25, twistLimitDegrees=1, twistMotion=Locked}
 {bone1Name=dyn_strap_03, bone2Name=dyn_strap_02, swing1LimitDegrees=25, swing2LimitDegrees=25, twistLimitDegrees=1, twistMotion=Locked}
]
